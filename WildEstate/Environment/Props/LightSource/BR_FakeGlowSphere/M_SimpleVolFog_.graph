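# Unlit additive material that fakes a volumetric glow or fog sphere around light props, with day and night color and brightness.
# A ray-traced sphere falloff gets depth fade, optional hemisphere masking, near/far distance fades and a hologram collection `ScreensEnabled` multiplier.
# That visibility value also drives an offset collapsing vertices to the object position when near zero.
# The graph also holds a second, card-projected analytic sphere falloff with its own fades.
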Material /WildEstate/Environment/Props/LightSource/BR_FakeGlowSphere/M_SimpleVolFog_
graph {
 properties {
  bAutomaticallySetUsageInEditor=false
  blendMode=BLEND_Additive
  pixelDepthOffsetMode=PDOM_Legacy
  shadingModel=MSM_Unlit
  usedShadingModels=Unlit
 }
 nodes [
  n1 = LocalPosition()
  n2 = Normalize(VectorInput=Transform(None=WorldPosition() {worldPositionShaderOffset=WPT_CameraRelative}.XYZ) {transformSourceType=TRANSFORMSOURCE_World, transformType=TRANSFORM_Local})
  n3 = TransformPosition(None=Constant3Vector()) {transformType=TRANSFORMPOSSOURCE_TranslatedWorld}
  n4 = Transform(None=NamedRerouteDeclaration(None=@n3) {name=Card Position, nodeColor=<rgba 0 1 0.5313 1>}) {transformSourceType=TRANSFORMSOURCE_World, transformType=TRANSFORM_Local}
  n5 = NamedRerouteDeclaration(None=Normalize(VectorInput=@n4)) {name=Card Vector, nodeColor=<rgba 0.2188 0 1 1>}
  n6 = Multiply(A=@n2, B=Divide(A=DotProduct(A=@n1.XYZ, B=@n5), B=DotProduct(A=@n2, B=@n5)))
  n7 = Divide(A=NamedRerouteUsage(), B=Distance(A=@n4, B=Constant())) {constA=95}
  n8 = OneMinus(None=Multiply(A=@n7, B=@n7))
  n9 = Divide(A=Multiply(A=Subtract(A=@n1.XYZ, B=@n6), B=SquareRoot(None=@n8)), B=NamedRerouteUsage()) {constB=95}
  n10 = MaterialFunctionCall(Float3=Multiply(A=@n9, B=@n9)) {materialFunction=&BreakOutFloat3Components}
  n11 = OneMinus(None=Add(A=Add(A=@n10.R, B=@n10.R), B=@n10.B))
  n12 = Power(Base=@n11, Exp=ScalarParameter() {defaultValue=2, parameterName=Glow Falloff}) {constExponent=2.2061}
  n13 = Clamp(None=ScalarParameter() {defaultValue=0.344, parameterName=Hemisphere Falloff}) {maxDefault=0.499}
  n14 = Subtract(A=@n9, B=Multiply(A=SquareRoot(None=@n11), B=NamedRerouteUsage()))
  n15 = Add(A=Multiply(A=ComponentMask(None=@n14) {b=true}) {constB=-0.5}, B=ScalarParameter() {defaultValue=0.516, parameterName=Hemisphere Position}) {constB=0.5}
  n16 = StaticSwitchParameter(True=SmoothStep(Min=@n13, Max=Subtract(B=@n13), Value=Saturate(None=@n15)) {constMax=0.6, constMin=0.4}, False=Constant() {r=1}) {parameterName=UseHemisphereFalloff}
  n17 = Distance(A=WorldPosition() {worldPositionShaderOffset=WPT_CameraRelative}.XYZ, B=Constant())
  n18 = Divide(
   A=Subtract(A=@n17, B=ScalarParameter() {defaultValue=8665.7285, parameterName=FarFade_Offset})
   B=ScalarParameter() {defaultValue=20000, parameterName=FarFade_Distance}
  )
  n19 = Subtract(A=@n17, B=ScalarParameter() {parameterName=NearFade_Offset})
  n20 = ScalarParameter() {defaultValue=1000, parameterName=NearFade_Distance}
  n21 = Multiply(A=OneMinus(None=Saturate(None=@n18)), B=Saturate(None=Divide(A=@n19, B=@n20)))
  DepthFade(
   Opacity=Multiply(A=Multiply(A=@n12, B=@n16), B=VertexInterpolator(VS=@n21).PS)
   FadeDistance=ScalarParameter() {defaultValue=50, parameterName=DepthFade}
  )
  n22 = Distance(A=ObjectPositionWS(), B=CameraPositionWS())
  n23 = Multiply(
   A=ScalarParameter() {defaultValue=74.2047, parameterName=Radius}
   B=MaterialFunctionCall() {materialFunction=&ObjectScale}.Scale X
  )
  n24 = ScalarParameter() {parameterName=NearFade_Offset}
  n25 = Divide(A=Subtract(A=@n22, B=Add(A=@n23, B=@n24)), B=ScalarParameter() {defaultValue=1000, parameterName=NearFade_Distance})
  n26 = ScalarParameter() {defaultValue=8665.7285, parameterName=FarFade_Offset}
  n27 = Saturate(None=Divide(A=Subtract(A=@n22, B=@n26), B=ScalarParameter() {defaultValue=20000, parameterName=FarFade_Distance}))
  n28 = Multiply(
   A=Multiply(A=Saturate(None=@n25), B=OneMinus(None=@n27))
   B=CollectionParameter() {collection=/Game/Environments/TechArt/Hologram/MPC_Hologram, parameterName=ScreensEnabled}
  )
  n29 = Subtract(A=ObjectPositionWS(), B=WorldPosition().XYZ)
  If(A=@n28, B=Constant() {r=0.01}, "A > B"=Constant3Vector(), "A == B"=@n29, "A < B"=@n29)
  LocalPosition()
  MaterialFunctionCall(Distance=ScalarParameter() {parameterName=CameraFadeDistance}) {materialFunction=&Near_Camera_Fade}
  n30 = WorldPosition()
  n31 = CameraPositionWS()
  n32 = MaterialFunctionCall!(
   Sphere Radius=Constant() {r=1}
   Sphere Falloff=@n23
   Unit Ray Direction=Normalize(VectorInput=Subtract(A=@n30.XYZ, B=@n31) {constB=-1})
   Sphere Center=ObjectPositionWS()
   Ray Origin=@n31
  ) {materialFunction=/Game/Athena/Environments/MaterialFunctions/RayTraceSphereFalloffMinClamp}
  n33 = DepthFade(
   Opacity=Power(Base=OneMinus(None=@n32.Clamped Falloff), Exp=ScalarParameter() {defaultValue=1.5, parameterName=FalloffFog})
   FadeDistance=ScalarParameter() {defaultValue=50, parameterName=DepthFade}
  )
  n34 = Constant3Vector() {constant=<rgba 0 0 -1 0>}
  n35 = Add(A=DotProduct(A=VertexNormalWS(), B=@n34), B=ScalarParameter() {group=Hemisphere_Falloff, parameterName=Hemisphere_Falloff_Position})
  n36 = Max(A=ScalarParameter() {defaultValue=0.0216, group=Hemisphere_Falloff, parameterName=Hemisphere_Falloff_Distance}) {constB=0}
  n37 = Saturate(None=Power(Base=Divide(A=@n35, B=@n36), Exp=ScalarParameter() {defaultValue=2, group=Hemisphere_Falloff, parameterName=Hemisphere_Falloff_Power}))
  Multiply(A=StaticSwitchParameter(True=Multiply(A=@n33, B=@n37), False=@n33) {parameterName=UseHemisphereFalloff}, B=@n28)
  n38 = Add(A=@n21) {constB=-0.01}
  Multiply(A=Add(A=Ceil(None=@n38)) {constB=-1}, B=Transform(None=LocalPosition().XYZ) {transformSourceType=TRANSFORMSOURCE_Local})
  n39 = ScalarParameter() {defaultValue=1, parameterName=Brightness_night}
  n40 = VectorParameter() {parameterName=Color_Night, defaultValue=<rgba 1 0.477 0 1>}
  Multiply(
   A=MaterialFunctionCall(Value_Morning=@n39, Value_Day=ScalarParameter() {defaultValue=0.5, parameterName=Brightness_Day}, Value_Evening=@n39, Value_Night=@n39) {materialFunction=&Scalar_Time_of_Day}.Result
   B=MaterialFunctionCall(
    Color_Morning=@n40.RGB
    Color_Day=VectorParameter() {parameterName=Color_Day, defaultValue=<rgba 0 0.0818 0.25 1>}.RGB
    Color_Evening=@n40.RGB
    Color_Night=@n40.RGB
   ) {materialFunction=&Color_Time_of_Day}.Result
  )
  n41 = ScalarParameter() {defaultValue=95, parameterName=Sphere Radius}
  NamedRerouteDeclaration(None=@n41) {name=Sphere Radius, nodeColor=<rgba 0.8516 0 1 1>}
  Reroute(None=@n41)
  Reroute(None=@n3)
  Transform(None=@n34) {transformSourceType=TRANSFORMSOURCE_Local}
 ]
}
